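# Physics asset for the Festive Pug Male two-handed pickaxe, giving its dangling meat flaps, pineapple pieces and ribbon bow secondary motion.
# Small master bodies join each group to the capsule handle with free angular constraints.
# Box bodies hang from them with limited swing and twist: a meat-flap chain forking upper and lower, three top/bottom pineapple pairs, and bow loops with two tail chains.
# The ribbon base also rotates freely on its master.
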PhysicsAsset /BRCosmetics/Weapons/FORT_Melee/Pickaxe_FestivePugMale/Meshes/Festive_Pug_Male_Physics
bodies {
 Melee_TwoHanded_Handle [
  {shapeType=Capsule, length=16.2459, radius=3.5, center=<0.936 -1.9834 37.2084>, rotation=<rot 0 0 7.3189>}
  {shapeType=Capsule, length=15, radius=4, center=<-0.009 0.5334 53.0267>, rotation=<rot 2.1834 0.8264 9.9151>}
 ]
 dyn_meatFlapBaseA [{shapeType=Box, extentX=10.53, extentY=6.3396, extentZ=39.3632, center=<2.3003 -0.3646 -0.0001>}]
 dyn_meatFlapBaseB [{shapeType=Box, extentX=4.6777, extentY=6.9549, extentZ=36.5335, center=<2.0948 -1.938 0.0003>, rotation=<rot 0 -29.9985 0>}]
 dyn_meatFlapLower1 [{shapeType=Box, extentX=3.6641, extentY=2.8794, extentZ=31.5822, center=<1.5471 0.1739 0.0584>, rotation=<rot 0 0 -5.0002>}]
 dyn_meatFlapLower2 [{shapeType=Box, extentX=2.6637, extentY=2.8512, extentZ=29.8506, center=<1.4874 -0.4149 -0.0447>}]
 dyn_meatFlapLower3 [{shapeType=Box, extentX=5.489, extentY=2.7688, extentZ=27.4994, center=<2.688 -0.1235 0.1637>}]
 dyn_meatFlapLower4 [{shapeType=Box, extentX=5.5, extentY=3, extentZ=23.5, center=<2.6349 0 -0.0001>}]
 dyn_meatFlapUpper1 [{shapeType=Box, extentX=5.379, extentY=3.4227, extentZ=29.6607, center=<2.7706 0.2292 -0.007>, rotation=<rot 0 -0.0001 0>}]
 dyn_meatFlapUpper2 [{shapeType=Box, extentX=5.0719, extentY=3.3246, extentZ=26.5255, center=<2.6709 0.0255 0.0001>}]
 dyn_meatFlapUpper3 [{shapeType=Box, extentX=3.9384, extentY=3.6573, extentZ=18.4819, center=<1.9602 0.2832 0.0002>}]
 dyn_mstr_meatFlapBase [
  {shapeType=Capsule, length=0.505, radius=0.505}
  {shapeType=Box, extentX=31.5, extentY=7.5, extentZ=35, center=<11.5164 -8.6964 0.0001>, rotation=<rot 0 -5.0006 0>}
 ]
 dyn_mstr_pineappleLeft [{shapeType=Capsule, length=0.505, radius=0.505}]
 dyn_mstr_pineappleRight [{shapeType=Capsule, length=0.505, radius=0.505}]
 dyn_mstr_pineappleUpper [{shapeType=Capsule, length=0.505, radius=0.505}]
 dyn_mstr_ribbonBase [{shapeType=Capsule, length=0.505, radius=0.505, rotation=<rot 0 -90 0>}]
 dyn_pineappleLeft_bottom [{shapeType=Box, extentX=8.7817, extentY=1.943, extentZ=16.9077, center=<1.0456 0.1798 -0.0646>}]
 dyn_pineappleLeft_top [{shapeType=Box, extentX=8.6345, extentY=2.0796, extentZ=16.9261, center=<0.9722 -0.1935 0.0021>}]
 dyn_pineappleRight_bottom [{shapeType=Box, extentX=8.5869, extentY=2.4716, extentZ=17.0695, center=<0.8546 0.0416 0.043>}]
 dyn_pineappleRight_top [{shapeType=Box, extentX=8.2525, extentY=2.0717, extentZ=16.3742, center=<1.2517 -0.0606 0.0986>}]
 dyn_pineappleUpper_bottom [{shapeType=Box, extentX=8.7276, extentY=2.3824, extentZ=16.4629, center=<1.0162 0.2511 0.0016>, rotation=<rot 0 -8.9004 0>}]
 dyn_pineappleUpper_top [{shapeType=Box, extentX=8.7962, extentY=1.6384, extentZ=16.8905, center=<1.0031 -0.1053 0.0657>, rotation=<rot -0.8869 0 0>}]
 dyn_ribbon1_l [{shapeType=Box, extentX=4.5, extentY=3, extentZ=2, center=<2.25 0 0.5>}]
 dyn_ribbon1_r [{shapeType=Box, extentX=4.5, extentY=3, extentZ=2, center=<2.25 0 0.5>}]
 dyn_ribbon2_l [{shapeType=Box, extentX=6.5, extentY=3.5, extentZ=2, center=<3.25 0 0.5>}]
 dyn_ribbon2_r [{shapeType=Box, extentX=7.5, extentY=3.2, extentZ=2, center=<3.75 0 0.5>}]
 dyn_ribbon3_l [{shapeType=Box, extentX=6.5, extentY=3.5, extentZ=2, center=<3.25 0 0>}]
 dyn_ribbon3_r [{shapeType=Box, extentX=8, extentY=3.3, extentZ=2, center=<4 0 0.5>}]
 dyn_ribbon4_r [{shapeType=Box, extentX=4, extentY=3.5, extentZ=2, center=<2 0 0.5>}]
 dyn_ribbonBase [{shapeType=Box, extentX=4.7869, extentY=3.7313, extentZ=4.2158, center=<0.8927 -0.3833 -1.8417>}]
 dyn_ribbonBow_l [{shapeType=Box, extentX=8.2556, extentY=9.8792, extentZ=7.0027, center=<4.8916 -0.0879 -0.9928>}]
 dyn_ribbonBow_r [{shapeType=Box, extentX=8.3074, extentY=9.8922, extentZ=6.9007, center=<5.1739 -0.3213 -0.089>}]
}
constraints [
 {bone1Name=dyn_ribbon2_l, bone2Name=dyn_ribbon1_l, swing1LimitDegrees=45, swing2LimitDegrees=45, twistLimitDegrees=15}
 {bone1Name=dyn_ribbon3_l, bone2Name=dyn_ribbon2_l, swing1LimitDegrees=45, swing2LimitDegrees=45, twistLimitDegrees=15}
 {
  bone1Name=dyn_ribbonBase
  bone2Name=dyn_mstr_ribbonBase
  swing1LimitDegrees=45
  swing1Motion=Free
  swing2LimitDegrees=45
  swing2Motion=Free
  twistLimitDegrees=45
  twistMotion=Free
 }
 {bone1Name=dyn_ribbon1_l, bone2Name=dyn_ribbonBase, swing1LimitDegrees=45, swing2LimitDegrees=45, twistLimitDegrees=15}
 {bone1Name=dyn_ribbon1_r, bone2Name=dyn_ribbonBase, swing1LimitDegrees=45, swing2LimitDegrees=45, twistLimitDegrees=15}
 {bone1Name=dyn_ribbon2_r, bone2Name=dyn_ribbon1_r, swing1LimitDegrees=45, swing2LimitDegrees=45, twistLimitDegrees=15}
 {bone1Name=dyn_ribbon3_r, bone2Name=dyn_ribbon2_r, swing1LimitDegrees=45, swing2LimitDegrees=45, twistLimitDegrees=15}
 {bone1Name=dyn_ribbon4_r, bone2Name=dyn_ribbon3_r, swing1LimitDegrees=45, swing2LimitDegrees=45, twistLimitDegrees=15}
 {bone1Name=dyn_ribbonBow_r, bone2Name=dyn_ribbonBase, swing1LimitDegrees=25, swing2LimitDegrees=25, twistLimitDegrees=10}
 {bone1Name=dyn_ribbonBow_l, bone2Name=dyn_ribbonBase, swing1LimitDegrees=25, swing2LimitDegrees=25, twistLimitDegrees=10}
 {bone1Name=dyn_meatFlapBaseA, bone2Name=dyn_mstr_meatFlapBase, swing1LimitDegrees=10, swing2LimitDegrees=10, twistLimitDegrees=5}
 {bone1Name=dyn_meatFlapBaseB, bone2Name=dyn_meatFlapBaseA, swing1LimitDegrees=25, swing2LimitDegrees=5, twistLimitDegrees=5}
 {bone1Name=dyn_meatFlapLower1, bone2Name=dyn_meatFlapBaseB, swing1LimitDegrees=10, swing2LimitDegrees=10, twistLimitDegrees=5}
 {bone1Name=dyn_meatFlapLower2, bone2Name=dyn_meatFlapLower1, swing1LimitDegrees=10, swing2LimitDegrees=10, twistLimitDegrees=10}
 {bone1Name=dyn_meatFlapLower3, bone2Name=dyn_meatFlapLower2, swing1LimitDegrees=13, swing2LimitDegrees=13, twistLimitDegrees=10}
 {bone1Name=dyn_meatFlapUpper1, bone2Name=dyn_meatFlapBaseB, swing1LimitDegrees=45, swing2LimitDegrees=5, twistLimitDegrees=5}
 {bone1Name=dyn_meatFlapUpper2, bone2Name=dyn_meatFlapUpper1, swing1LimitDegrees=25, swing2LimitDegrees=4, twistLimitDegrees=5}
 {bone1Name=dyn_meatFlapUpper3, bone2Name=dyn_meatFlapUpper2, swing1LimitDegrees=10, swing2LimitDegrees=5, twistLimitDegrees=5}
 {bone1Name=dyn_meatFlapLower4, bone2Name=dyn_meatFlapLower3, swing1LimitDegrees=10, swing2LimitDegrees=10, twistLimitDegrees=5}
 {bone1Name=dyn_pineappleRight_top, bone2Name=dyn_mstr_pineappleRight, swing1LimitDegrees=10, swing2LimitDegrees=5, twistLimitDegrees=1}
 {bone1Name=dyn_pineappleRight_bottom, bone2Name=dyn_mstr_pineappleRight, swing1LimitDegrees=10, swing2LimitDegrees=5, twistLimitDegrees=1}
 {bone1Name=dyn_pineappleUpper_top, bone2Name=dyn_mstr_pineappleUpper, swing1LimitDegrees=10, swing2LimitDegrees=5, twistLimitDegrees=1}
 {bone1Name=dyn_pineappleUpper_bottom, bone2Name=dyn_mstr_pineappleUpper, swing1LimitDegrees=10, swing2LimitDegrees=5, twistLimitDegrees=1}
 {bone1Name=dyn_pineappleLeft_top, bone2Name=dyn_mstr_pineappleLeft, swing1LimitDegrees=10, swing2LimitDegrees=5, twistLimitDegrees=1}
 {bone1Name=dyn_pineappleLeft_bottom, bone2Name=dyn_mstr_pineappleLeft, swing1LimitDegrees=10, swing2LimitDegrees=5, twistLimitDegrees=1}
 {
  bone1Name=dyn_mstr_meatFlapBase
  bone2Name=Melee_TwoHanded_Handle
  swing1LimitDegrees=45
  swing1Motion=Free
  swing2LimitDegrees=45
  swing2Motion=Free
  twistLimitDegrees=45
  twistMotion=Free
 }
 {
  bone1Name=dyn_mstr_ribbonBase
  bone2Name=Melee_TwoHanded_Handle
  swing1LimitDegrees=45
  swing1Motion=Free
  swing2LimitDegrees=45
  swing2Motion=Free
  twistLimitDegrees=45
  twistMotion=Free
 }
 {
  bone1Name=dyn_mstr_pineappleRight
  bone2Name=Melee_TwoHanded_Handle
  swing1LimitDegrees=45
  swing1Motion=Free
  swing2LimitDegrees=45
  swing2Motion=Free
  twistLimitDegrees=45
  twistMotion=Free
 }
 {
  bone1Name=dyn_mstr_pineappleUpper
  bone2Name=Melee_TwoHanded_Handle
  swing1LimitDegrees=45
  swing1Motion=Free
  swing2LimitDegrees=45
  swing2Motion=Free
  twistLimitDegrees=45
  twistMotion=Free
 }
 {
  bone1Name=dyn_mstr_pineappleLeft
  bone2Name=Melee_TwoHanded_Handle
  swing1LimitDegrees=45
  swing1Motion=Free
  swing2LimitDegrees=45
  swing2Motion=Free
  twistLimitDegrees=45
  twistMotion=Free
 }
]
